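# Physics asset for the demo skeletal mesh of the Snowman two-handed pickaxe, a Battle Royale cosmetic melee weapon.
# It holds a single body on the Melee_TwoHanded_Handle bone, shaped as one rotated capsule offset along the weapon.
# There are no constraints, so no joints link any bodies.
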PhysicsAsset /BRCosmetics/Weapons/FORT_Melee/Pickaxe_Snowman/Meshes/DEMO/SK_Pickaxe_Snowman_DEMO_PhysicsAsset
bodies {Melee_TwoHanded_Handle [{shapeType=Capsule, length=71.8748, radius=35.6457, center=<-0.1657 1.5837 23.4986>, rotation=<rot -0.3043 -90 -0.2003>}]}
constraints []
